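; BP_EmitterDisplay is a display actor in the NeverCook FX gym that shows a Niagara system with a text label.
; The construction script assigns the NiagaraAsset variable to the Niagara component, then sets the TextRender to that asset's object name.
; An Open Confluence custom event launches a wiki page on Niagara Day Sequence time of day.
; BeginPlay, Tick and overlap events are empty.
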(event EventBeginPlay)

(event Collision|EventActorBeginOverlap
  (params (OtherActor :type Actor%20Object%20Reference)))

(event EventTick
  (params (DeltaSeconds :type Float%20%28single-precision%29)))

(event Custom|Open%20Confluence
  (Utilities|Platform|LaunchURL "https://epicgames.atlassian.net/wiki/spaces/FORT/pages/453279804/Niagara+Day+Sequence+Time+Of+Day"))

(fn ConstructionScript
  (Niagara|SetNiagaraSystemAsset (Variables|Default|GetNiagara) (Variables|Default|GetNiagaraAsset) true)
  (Rendering|Components|TextRender|SetText (Variables|Default|GetTextRender) (Utilities|Text|ToText(String) (Utilities|GetObjectName (Niagara|GetNiagaraSystemAsset (Variables|Default|GetNiagara))))))
