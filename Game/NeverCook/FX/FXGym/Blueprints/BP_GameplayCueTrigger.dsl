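; BP_GameplayCueTrigger is an FX gym trigger actor that plays a configured gameplay cue on any actor overlapping its Box.
; In burst mode it executes the cue once on overlap; otherwise it adds a looping cue on begin overlap and removes it on end overlap.
; The construction script writes the gameplay cue tag into a text render component as a label.
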(event EventBeginPlay)

(event Collision|EventActorBeginOverlap
  (params (OtherActor :type Actor%20Object%20Reference)))

(event EventTick
  (params (DeltaSeconds :type Float%20%28single-precision%29)))

(event OnComponentBeginOverlap(Box)
  (params (OverlappedComponent :type Primitive%20Component%20Object%20Reference) (OtherActor :type Actor%20Object%20Reference) (OtherComp :type Primitive%20Component%20Object%20Reference) (OtherBodyIndex :type Integer) (bFromSweep :type Boolean) (SweepResult :type Hit%20Result%20Structure%20%28by%20ref%29))
  (bind _returnvalue (Ability|GameplayCue|MakeGameplayCueParameters 0.0))
  (bind _gameplay_cue_tag (Variables|Default|GetGameplay%20Cue%20Tag))
  (if (Variables|Default|GetbisBurst?)
    (GameplayCue|ExecuteGameplayCueOnActor(Burst) OtherActor _gameplay_cue_tag _returnvalue)
    (else
      (GameplayCue|AddGameplayCueOnActor(Looping) OtherActor _gameplay_cue_tag _returnvalue))))

(event OnComponentEndOverlap(Box)
  (params (OverlappedComponent :type Primitive%20Component%20Object%20Reference) (OtherActor :type Actor%20Object%20Reference) (OtherComp :type Primitive%20Component%20Object%20Reference) (OtherBodyIndex :type Integer))
  (if (not (Variables|Default|GetbisBurst?))
    (GameplayCue|RemoveGameplayCueOnActor(Looping) OtherActor (Variables|Default|GetGameplay%20Cue%20Tag) (Ability|GameplayCue|MakeGameplayCueParameters 0.0))))

(event InteractInterface|EventOnInteractFocused
  (params (FortPC :type Fort%20Player%20Controller%20Object%20Reference) (ContextInfo :type Fort%20Interact%20Context%20Info%20Object%20Reference)))

(fn ConstructionScript
  (Rendering|Components|TextRender|SetText (Variables|Default|GetTextRender) (Class|QuestDebugToolBPHelpers|GameplayTagtoText (Variables|Default|GetGameplay%20Cue%20Tag))))
