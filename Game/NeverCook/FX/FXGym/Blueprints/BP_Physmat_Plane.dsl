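; BP_Physmat_Plane is an FX gym display actor that shows a material on a plane, labelled with that material's physical material.
; The construction script applies the Mat variable to slot 0 of the Plane component, then reads back that material's physical material and writes it as text to the TextRender.
; The label is therefore derived from the material itself at construction.
; BeginPlay, Tick and overlap events are empty.
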(event EventBeginPlay)

(event Collision|EventActorBeginOverlap
  (params (OtherActor :type Actor%20Object%20Reference)))

(event EventTick
  (params (DeltaSeconds :type Float%20%28single-precision%29)))

(fn ConstructionScript
  (bind _plane (Variables|Default|GetPlane))
  (Rendering|Material|SetMaterial _plane 0 (Variables|Default|GetMat))
  (Rendering|Components|TextRender|SetText (Variables|Default|GetTextRender) (Utilities|Text|ToText(Object) (Physics|Material|GetPhysicalMaterial (Rendering|Material|GetMaterial _plane 0)))))
