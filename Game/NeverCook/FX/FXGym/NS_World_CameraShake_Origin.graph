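# NS_World_CameraShake_Origin is a Niagara system in the FX Gym folder, named as a world camera shake origin.
# Its single Minimal emitter uses an instantaneous burst to spawn particles with no position offset.
# A component renderer is the only renderer, so particles appear through spawned components rather than sprites or meshes; the component class is not shown.
# Particles are not killed when their lifetime elapses.
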NiagaraSystem /Game/NeverCook/FX/FXGym/NS_World_CameraShake_Origin
dependencies {usedDataInterfaces [], usedDynamicInputs [], usedModules [&SystemState]}
emitters {Minimal {
  emitterUpdateScript [
   &EmitterState()
   &SpawnBurst_Instantaneous(Use Spawn Probability=false, UseLoopCountLimit=true)
  ]
  particleSpawnScript [&InitializeParticle(Color Mode=::NewEnumerator0, Lifetime=5, UsePositionOffset=false, Write Material Random=true)]
  particleUpdateScript [&ParticleState(Kill Particles When Lifetime Has Elapsed=false, Let Infinitely Lived Particles Die When Emitter Deactivates=true)]
  rendererClasses [&NiagaraComponentRendererProperties]
  renderers [{rendererClass=&NiagaraComponentRendererProperties, rendererIndex=0}]
 }}
summary {
 systemName=NS_World_CameraShake_Origin
 emitters [{emitterName=Minimal, rendererClasses [&NiagaraComponentRendererProperties]}]
 userVariables []
}
